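# Physics asset for the SilkWalker semi truck prototype, giving the vehicle its simplified collision shape.
# A single frame body combines boxes and capsules, with no constraints.
# A large base box spans the chassis, with further shapes for the seat, the area behind it and the rear floor.
# Three capsules named Winter_Mustache trace the front end as a crossbar with two angled side pieces.
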PhysicsAsset /Valet/BigRig/Proto/Semi_Truck_SilkWalker_Physics
bodies {frame [
  {shapeType=Capsule, length=80, radius=45, shapeName=BackSeatCollision, center=<-275.9377 -0.0006 138.7764>}
  {
   shapeType=Capsule
   length=33.5
   radius=15
   shapeName=Winter_Mustache_r
   center=<337.4694 150.7477 32.9952>
   rotation=<rot 13.5662 6.461 25.7669>
  }
  {
   shapeType=Capsule
   length=33.5
   radius=15
   shapeName=Winter_Mustache_l
   center=<336.8204 -140.9383 34.916>
   rotation=<rot 18.1155 4.3055 -24.8216>
  }
  {shapeType=Capsule, length=259.5, radius=12, shapeName=Winter_Mustache_mid, center=<343.5607 -0.0002 17.9123>, rotation=<rot 0 0 -90.0002>}
  {shapeType=Box, extentX=704, extentY=241, extentZ=67, shapeName=PhysicsBody_Base, center=<-25.4432 -0.0024 5.3731>}
  {shapeType=Box, extentX=80.5, extentY=252.5, extentZ=118, shapeName=Winter_SeatBase, center=<-107.476 0.0002 97.5451>}
  {shapeType=Box, extentX=17, extentY=137, extentZ=155.5, shapeName=Winter_SeatBack, center=<-75.9118 0.001 225.6471>}
  {
   shapeType=Box
   extentX=14.5
   extentY=89.5
   extentZ=83
   shapeName=Winter_SeatTop
   center=<-77.3121 3.1101 274.8778>
   rotation=<rot 0 0 45.0001>
  }
  {shapeType=Box, extentX=65, extentY=252.5, extentZ=175.5, shapeName=Winter_BehindSeat_Phys, center=<-36.1775 -0.0002 127.1735>}
  {shapeType=Box, extentX=370.5, extentY=255.5, extentZ=20, shapeName=Winter_RearFloor, center=<-220.1175 -0.003 47.818>}
 ]}
constraints []
