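# Prototype BigRig passenger animation layer Blueprint for the Fortnite player skeleton, implementing vehicle override pose layers.
# Only VehicleSplitBodyOverride has wired logic: it applies the FerretVehicle passenger slope aim offset, driven by slope roll and pitch angles, to the root-removed upper-and-lower-body input pose.
# The final pose, full-body and lower-body layers leave their outputs unconnected, and the event graph holds only unwired default nodes.
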AnimBlueprint /Valet/BigRig/Proto/Silkwalker_PassengerLayer_AnimBP
target_skeleton=&Fortnite_M_Avg_Player_Skeleton
graphs [
 AnimGraph {nodes [Output Pose() {layerGroup=None}]}
 EventGraph {nodes [
   TryGetPawnOwner()
   Event BlueprintUpdateAnimation()
  ]}
 VehicleFinalPoseOverride {nodes [
   Input Pose() {name=InFinalPose}
   Output Pose() {layerGroup=VehicleOverride}
  ]}
 VehicleFullBodyOverride {nodes [
   Input Pose() {name=InPoseFullBody}
   Output Pose() {layerGroup=VehicleOverride}
  ]}
 VehicleLowerBodyOverride {nodes [
   Input Pose() {name=InPoseLowerBodyDefault}
   Input Pose() {name=InPoseUpperBody}
   Output Pose() {layerGroup=VehicleOverride}
  ]}
 VehicleSplitBodyOverride {nodes [
   Input Pose() {name=InPoseSplitBody}
   Input Pose() {name=InPoseUpperAndLowerBody}
   n1 = AimOffset Player 'FerretVehicle_PassengerSlope_AO'!(
    X:float=Get SlopeRollDegreeAngle() -> (SlopeRollDegreeAngle:float)
    Y:float=Get SlopePitchDegreeAngle() -> (SlopePitchDegreeAngle:float)
    BasePose:Pose Link Structure=Input Pose() {name=InPoseUpperAndLowerBodyRemoveRoot}
    Alpha:float=1
   ) -> (Pose:Pose Link Structure) {blendSpace=/Game/Animation/Game/MainPlayer/Vehicles/FerretVehicle/AO/FerretVehicle_PassengerSlope_AO, alphaBoolBlend {blendOption=Linear}}
   Output Pose(Result=@n1) {layerGroup=Vehicle Override}
  ]}
]
variables []
