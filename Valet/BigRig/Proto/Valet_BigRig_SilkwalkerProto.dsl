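; Prototype Silkwalker big rig vehicle Blueprint that drives music state and handles idle, boost exhaust and explosion fade effects.
; It registers with the gameplay data tracker manager, mapping alert levels to tags.
; The seat 2 bot's alert changes and death, plus a player-occupant array, update that state, using "Count" to stop music.
; Damaging it while a flagger-tagged passenger is aboard gives the instigating player a negative tag.
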(event EventPickRandomMaterial)

(event EventActivateBoostFX
  ;[1 1 1] Activate Boost Exhaust FX
  (bind _returnvalue (Niagara|SpawnSystemAttached "/Valet/BigRig/Proto/NS_SgtWinter_TruckExhaust_RedSmoke.NS_SgtWinter_TruckExhaust_RedSmoke" (Variables|Vehicle|GetSkeletalMesh) "root" "0, 0, 0" "0, 0, 0" "SnapToTarget" true))
  (Variables|Default|SetFX_Boost_Exhaust_R _returnvalue))

(event EventDeactivateBoostFX
  ;[1 1 1] Deactivate Boost FX
  (Components|Activation|Deactivate (Variables|Default|GetFX_Boost_Exhaust_R)))

(event EventBeginPlay
  (|CallParentFunction)
  (CallFunction|BeginMusicTracker)
  (Variables|Default|SetIsDestroyed false)
  (bind _returnvalue (Niagara|SpawnSystemAttached "/Valet/BigRig/Proto/NS_SgtWinter_Truck_Idle.NS_SgtWinter_Truck_Idle" (Variables|Vehicle|GetSkeletalMesh) "root" "0, 0, 0" "0, 0, 0" "SnapToTarget" true))
  (Variables|Default|SetIdle_VFX _returnvalue))

(event EventEndPlay
  (params (EndPlayReason :type EEndPlayReason%20Enum :default "Destroyed"))
  (|CallParentFunction EndPlayReason)
  (CallFunction|EndMusicTracker)
  (CallFunction|SetAlertLevel "Count")
  (Variables|Default|SetIsDestroyed true))

(event EventOnDeathPlayEffects
  (params (Damage :type Float%20%28single-precision%29) (DamageTags :type Gameplay%20Tag%20Container%20Structure%20%28by%20ref%29) (Momentum :type Vector :default "0, 0, 0") (HitInfo :type Hit%20Result%20Structure%20%28by%20ref%29) (InstigatedBy :type Fort%20Pawn%20Object%20Reference) (DamageCauser :type Actor%20Object%20Reference) (EffectContext :type Gameplay%20Effect%20Context%20Handle%20Structure))
  (|CallParentFunction Damage DamageTags Momentum HitInfo InstigatedBy DamageCauser EffectContext)
  (CallFunction|EndMusicTracker)
  (CallFunction|SetAlertLevel "Count")
  (Variables|Default|SetIsDestroyed true))

(event Custom|End%20Music%20Tracker
  ;[1 0.9442 0.1824 0.3] Music Tracker
  (bind _returnvalue (FortGameplayDataTrackerComponent|GetFortGameplayDataTrackerComponentManager))
  (Utilities|IsValid _returnvalue
    (:Is%20Valid
      (Class|FortGameplayDataTrackerComponentManager|UnRegisterTrackedStateActor _returnvalue self)
      (Utilities|Array|Clear (Variables|Default|GetPlayer%20Pawn)))))

(event Custom|BeginMusicTracker
  (bind _alertleveltag (Variables|Default|GetAlertLevelTag))
  (bind (_value _returnvalue) (Utilities|Map|Find _alertleveltag "Count"))
  ;[1 0.9442 0.1824 0.3] Music Tracker
  (bind _returnvalue_2 (FortGameplayDataTrackerComponent|GetFortGameplayDataTrackerComponentManager))
  (Utilities|IsValid _returnvalue_2
    (:Is%20Valid
      (if _returnvalue
        (Class|FortGameplayDataTrackerComponentManager|RegisterTrackedStateActor _returnvalue_2 self _value)))))

(event Custom|SetMusicTrackerState
  (params (AlertLevel :type EAlertLevel%20Enum :default "Unaware"))
  ;[1 0.9442 0.1824 0.3] Music Tracker
  (CallFunction|SetAlertLevel AlertLevel))

(event Custom|PlayerEnterVehicleUpdate
  (params (PlayerPawn :type Fort%20Player%20Pawn%20Object%20Reference))
  ;[1 0.9442 0.1824 0.3] Music Tracker
  ;;[1 0.9442 0.1824 0.3] Add Players to array to start music back up on entering vehicle. Then only play it once when first player enters.
  (if (Pawn|IsPlayerControlled PlayerPawn)
    (Utilities|Array|Add (Variables|Default|GetPlayer%20Pawn) PlayerPawn)
    (Utilities|FlowControl|DoOnce false
      (:Completed
        (CallFunction|SetAlertLevel "Unaware")))))

(event Custom|ResetMusicEnterUpdate
  ;[1 0.9442 0.1824 0.3] Music Tracker
  ;;[1 0.9442 0.1824 0.3] Add Players to array to start music back up on entering vehicle. Then only play it once when first player enters.
  (Utilities|FlowControl|DoOnce false
    (:Completed
      (CallFunction|SetAlertLevel "Unaware"))))

(event EventExplodeCosmetics
  (params (DestroyedActor :type Actor%20Object%20Reference))
  (bind _semi_truck_winter (Variables|Default|GetSemi_Truck_Winter))
  ;[1 1 1] EVENT -> Explode Cosmetics: Set Material on attachment so it fades out with the rest of the vehicle.
  (for (_ _array_index) (Rendering|Material|GetMaterials _semi_truck_winter)
    (Rendering|Material|SetMaterial _semi_truck_winter _array_index "/Valet/Materials/M_FN_Valet_Destroyed_Master_InstAdjusted.M_FN_Valet_Destroyed_Master_InstAdjusted")
    (|CallParentFunction DestroyedActor)
    (Components|Activation|Deactivate (Variables|Default|GetIdle_VFX))))

(event EventOnTickAudioFade
  (params (FadeParam :type Float%20%28single-precision%29))
  ;[1 1 1] EVENT -> Explode Cosmetics: Set Material on attachment so it fades out with the rest of the vehicle.
  (|CallParentFunction FadeParam)
  (Rendering|Material|SetScalarParameterValueonMaterials (Variables|Default|GetSemi_Truck_Winter) "Fade" FadeParam))

(event EventDisolveVehicle
  (params (Fade :type Float%20%28double-precision%29)))

(event EventOnPawnExitVehicle
  (params (PlayerPawn :type Fort%20Player%20Pawn%20Object%20Reference) (ExitSocketName :type Name :default "None"))
  ;[0.5277 1 0.6739] EVENT -> Pawn Exit Vehicle
  (|CallParentFunction PlayerPawn ExitSocketName)
  ;;[1 0.9442 0.1824 0.3] Check to see if we should stop music as players exit vehicle
  (CallFunction|PlayerExitVehicleUpdate PlayerPawn))

(event EventOnPawnEnterVehicle
  (params (PlayerPawn :type Fort%20Player%20Pawn%20Object%20Reference) (SeatIdx :type Integer))
  ;[0.5277 1 0.6739] EVENT -> Pawn Enter Vehicle
  (if (== SeatIdx 2)
    (bind _asfort_athena_aibot_controller (Utilities|Casting|CastToFortAthenaAIBotController (Pawn|GetController PlayerPawn)))
    (AlertLevel|AssignOnAlertLevelChangedEvent _asfort_athena_aibot_controller (EventDispatchers|CreateEvent :!event Custom|OnSilkwalkerAlertLevelChanged))
    (bind _output_get (Variables|Default|SetSilkwalkerPawn PlayerPawn))
    (EventDispatchers|BindEventtoOnDied _output_get (EventDispatchers|CreateEvent :!event Custom|On%20Died))
    (else
      (CallFunction|PlayerEnterVehicleUpdate PlayerPawn))))

(event Custom|OnSilkwalkerAlertLevelChanged
  (params (BotController :type Fort%20Athena%20AIBot%20Controller%20Object%20Reference) (OldAlertLevel :type EAlertLevel%20Enum :default "Unaware") (NewAlertLevel :type EAlertLevel%20Enum :default "Unaware"))
  ;[0.5277 1 0.6739] EVENT -> On Silkwalker Alert Level Changed
  ;;[1 0.9442 0.1824 0.3] Update music based on SilkWalker Pawn alert state
  (CallFunction|SetAlertLevel NewAlertLevel))

(event Custom|On%20Died
  (params (DamagedActor :type Actor%20Object%20Reference) (Damage :type Float%20%28single-precision%29) (InstigatedBy :type Controller%20Object%20Reference) (DamageCauser :type Actor%20Object%20Reference) (HitLocation :type Vector) (FHitComponent :type Primitive%20Component%20Object%20Reference) (BoneName :type Name) (Momentum :type Vector))
  ;[0.5277 1 0.6739] EVENT -> On Silkwalker Died
  ;;[1 0.9442 0.1824 0.3] Stop Music on Silkwalker Death
  (CallFunction|SilkWalkerDeathAlert))

(event EventOnDamaged
  (params (Damage :type Float%20%28single-precision%29) (DamageTags :type Gameplay%20Tag%20Container%20Structure%20%28by%20ref%29) (EffectContext :type Gameplay%20Effect%20Context%20Handle%20Structure%20%28by%20ref%29) (EventInstigator :type Controller%20Object%20Reference) (DamageCauser :type Actor%20Object%20Reference))
  (bind _asfort_player_controller (Utilities|Casting|CastToFortPlayerController EventInstigator))
  (bind _array_element (Utilities|Array|ForEachLoopwithBreak (Vehicle|GetAllPassengers))
    (:LoopBody
      (if (GameplayTags|HasMatchingGameplayTag _array_element (Variables|Default|GetSilkwalkerFlaggerTag))
        (Ability|GameplayEffect|AddLooseGameplayTags _asfort_player_controller (Variables|Default|GetNegativeTag))))))

(fn PlayerExitVehicleUpdate
  (params (Player%20Pawn :type Fort%20Player%20Pawn%20Object%20Reference))
  (bind _player_pawn (Variables|Default|GetPlayer%20Pawn))
  ;[1 0.9442 0.1824 0.3] Check array with controlled players. Remove the cure pawn that enters from array. When array (vehicle) is empty then stop music, reset do once that plays music, and clear array (just incase)
  (if (Pawn|IsPlayerControlled Player%20Pawn)
    (bind _returnvalue (Utilities|Array|RemoveItem _player_pawn Player%20Pawn))
    (Development|PrintString (Utilities|String|ToString(Boolean) _returnvalue))
    (if (Utilities|Array|IsEmpty _player_pawn)
      (CallFunction|SetAlertLevel "Count")
      (CallFunction|ResetMusicEnterUpdate)
      (Utilities|Array|Clear (Variables|Default|GetPlayer%20Pawn)))))

(fn SetAlertLevel
  (params (AlertLevel :type EAlertLevel%20Enum :default "Unaware"))
  (bind _alertleveltag (Variables|Default|GetAlertLevelTag))
  (bind (_value _returnvalue) (Utilities|Map|Find _alertleveltag AlertLevel))
  (bind _returnvalue_2 (FortGameplayDataTrackerComponent|GetFortGameplayDataTrackerComponentManager))
  (Utilities|IsValid _returnvalue_2
    (:Is%20Valid
      (if _returnvalue
        (Class|FortGameplayDataTrackerComponentManager|SetTrackedActorState _returnvalue_2 self _value)))))

(fn SilkWalkerDeathAlert
  (if (and (Utilities|Array|IsEmpty (Variables|Default|GetPlayer%20Pawn)) (not (Variables|Default|GetIsDestroyed)))
    (CallFunction|SetAlertLevel "Count")))

(fn ConstructionScript
  (|CallParentFunction))
